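# M_PickaxeAnim is a Default Lit, Nanite-enabled material for the WildEstate quest vending machine, showing an animated hit icon, presumably a pickaxe, in screen style.
# Time drives a Frac and Floor step that shifts scaled, clamped UVs between two flipbook frames each second.
# The flipbook's red channel blends the Color1 and Color2 parameters.
# Separate branches build faint panning scanlines, tiled via HeightScale, and a Color2-tinted radial SphereMask.
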Material /WildEstate/Environment/Props/Quest/Quest_WEVendingMachine_A/Materials/M_PickaxeAnim
graph {
 properties {bUsedWithNanite=true, pixelDepthOffsetMode=PDOM_Legacy, usedShadingModels=Default Lit}
 nodes [
  Constant() {r=0.9}
  Constant() {r=0.2}
  Constant()
  n1 = VectorParameter() {parameterName=Color2, defaultValue=<rgba 0.6548 0.776 0.6987 0>}
  n2 = TextureCoordinate()
  n3 = Constant2Vector() {g=1.2, r=0.5}
  n4 = Add(A=Frac(None=Time())) {constB=0.5}
  n5 = Add(
   A=Add(A=Multiply(A=@n2, B=@n3) {constB=0.5}, B=Constant2Vector() {g=-0.07, r=-0.012})
   B=Multiply(A=Floor(None=@n4), B=Constant2Vector() {r=0.5}) {constB=0.5}
  )
  LinearInterpolate(
   A=VectorParameter() {parameterName=Color1, defaultValue=<rgba 0.0846 0.1302 0.1004 0>}.RGB
   B=@n1.RGB
   Alpha=TextureSampleParameter2D(UVs=Saturate(None=@n5)) {parameterName=Flipbook, texture=/Game/Creative/Items/Gameplay/VendingMachine/FX/T_VendingMachineHitIcon}.R
  )
  n6 = Multiply(
   A=TextureCoordinate() {vTiling=10}
   B=MaterialFunctionCall(X=Constant() {r=1}, Y=ScalarParameter() {defaultValue=2, parameterName=HeightScale}) {materialFunction=&MakeFloat2}.Result
  )
  Multiply(A=TextureSample(UVs=Panner(Coordinate=@n6) {speedX=0.5, speedY=0.7}) {texture=/Game/Missions/Primary/LaunchTheBalloon/Textures/T_TV_Scanlines}.RGB) {constB=0.05}
  Multiply(A=@n1.RGB, B=SphereMask(A=TextureCoordinate(), B=Constant() {r=0.5}, Radius=Constant() {r=0.8}, Hardness=Constant()))
 ]
}
